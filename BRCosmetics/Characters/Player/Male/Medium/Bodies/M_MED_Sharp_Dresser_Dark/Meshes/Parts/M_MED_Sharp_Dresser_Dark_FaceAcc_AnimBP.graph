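# Anim Blueprint for the Sharp Dresser Dark male face accessory mesh, fitting it to the character's head, jaw and neck.
# It copies pose and curves from the attached parent mesh, then chains bone offsets, scales and weighted multi-bone constraints on hat, neck, jaw, nose bridge and head.
# One left-neck constraint node sits outside the output chain.
# The event graph sets LOD_alpha when the current LOD exceeds 1, with no visible reader.
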AnimBlueprint /BRCosmetics/Characters/Player/Male/Medium/Bodies/M_MED_Sharp_Dresser_Dark/Meshes/Parts/M_MED_Sharp_Dresser_Dark_FaceAcc_AnimBP
target_skeleton=&/M_MED_Sharp_Dresser_Dark_FaceAcc_Skeleton
graphs [
 AnimGraph {nodes [
   "Constraint - dyn_left_neck_fix_mstr (PPPP)"!(
    ComponentPose
    Alpha=1
    ConstraintWeights_0=0.0684
    ConstraintWeights_1:float=0.8124
    ConstraintWeights_2:float=0.0269
    ConstraintWeights_3:float=0.0923
   ) -> (Pose) {
    alphaBoolBlend {blendOption=Linear}
    boneToModify {boneName=dyn_left_neck_fix_mstr}
    constraintSetup [{targetBone {boneName=C_jaw}}, {targetBone {boneName=head}}, {targetBone {boneName=neck_02}}, {targetBone {boneName=dyn_jaw_fix}}]
    constraintWeights [1,1,1,1]
   }
   n1 = "Transform (Modify) Bone - Bone: dyn_hat_fix"(
    ComponentPose=Local To Component(LocalPose=Copy Pose From Mesh() {bCopyCurves=true, bUseAttachedParent=true})
    Alpha=1
    Translation="0,0,0.5"
    Rotation="0,0,0"
    Scale="1,1,1"
   ) {
    scaleMode=BMM_Replace
    scaleSpace=BCS_BoneSpace
    translationMode=BMM_Replace
    translationSpace=BCS_BoneSpace
    alphaBoolBlend {blendOption=Linear}
    boneToModify {boneName=dyn_hat_fix}
    translation {z=0.5}
   }
   n2 = "Transform (Modify) Bone - Bone: dyn_neck_fix"(ComponentPose=@n1, Alpha=1, Translation="0,0,0", Rotation="0,0,0", Scale="1.1,1.1,1.1") {
    scaleMode=BMM_Replace
    scaleSpace=BCS_BoneSpace
    translationMode=BMM_Replace
    translationSpace=BCS_BoneSpace
    alphaBoolBlend {blendOption=Linear}
    boneToModify {boneName=dyn_neck_fix}
    scale=<1.1 1.1 1.1>
   }
   n3 = "Transform (Modify) Bone - Bone: dyn_jaw_fix"(ComponentPose=@n2, Alpha=1, Translation="0,0,0", Rotation="0,0,0", Scale="1.15,1,1") {
    scaleMode=BMM_Replace
    scaleSpace=BCS_BoneSpace
    translationMode=BMM_Replace
    translationSpace=BCS_BoneSpace
    alphaBoolBlend {blendOption=Linear}
    boneToModify {boneName=dyn_jaw_fix}
    scale {x=1.15}
   }
   n4 = "Constraint - C_nose_bridge (PP)"!(ComponentPose=@n3, Alpha=1, ConstraintWeights_0=0.2, ConstraintWeights_1:float=0.8) -> (Pose) {
    alphaBoolBlend {blendOption=Linear}
    boneToModify {boneName=C_nose_bridge}
    constraintSetup [{targetBone {boneName=faceAttach}}, {targetBone {boneName=head}}]
    constraintWeights [1,1]
   }
   n5 = "Transform (Modify) Bone - Bone: C_nose_bridge"!(ComponentPose=@n4, Alpha=1, Translation="0,0.25,0") -> (Pose) {
    scaleMode=BMM_Replace
    scaleSpace=BCS_BoneSpace
    translationMode=BMM_Replace
    translationSpace=BCS_BoneSpace
    alphaBoolBlend {blendOption=Linear}
    boneToModify {boneName=C_nose_bridge}
    translation {y=0.25}
   }
   n6 = "Transform (Modify) Bone - Bone: head"!(ComponentPose=@n5, bAlphaBoolEnabled:bool=True, Translation="-1.4,0.2,0", Rotation="0,0,0", Scale="1.06,1.05,1") -> (Pose) {
    alphaInputType=Bool
    scaleMode=BMM_Replace
    scaleSpace=BCS_BoneSpace
    translationMode=BMM_Replace
    translationSpace=BCS_BoneSpace
    alphaBoolBlend {blendOption=Linear}
    boneToModify {boneName=head}
    scale=<1.06 1.05>
    translation=<-1.4 0.2>
   }
   n7 = "Constraint - dyn_neck_fix_left_mstr (PPPPPPPP)"!(
    ComponentPose=@n6
    Alpha=1
    ConstraintWeights_0=0
    ConstraintWeights_1:float=0.0869
    ConstraintWeights_2:float=0.0002
    ConstraintWeights_3:float=0.0311
    ConstraintWeights_4:float=0.3016
    ConstraintWeights_5:float=0.2433
    ConstraintWeights_6:float=0.3368
    ConstraintWeights_7:float=0
   ) -> (Pose) {
    alphaBoolBlend {blendOption=Linear}
    boneToModify {boneName=dyn_neck_fix_left_mstr}
    constraintSetup [
     {targetBone {boneName=spine_04}}
     {targetBone {boneName=spine_05}}
     {targetBone {boneName=clavicle_l}}
     {targetBone {boneName=C_jaw}}
     {targetBone {boneName=neck_01}}
     {targetBone {boneName=head}}
     {targetBone {boneName=neck_02}}
     {targetBone {boneName=dyn_jaw_fix}}
    ]
    constraintWeights [1,1,1,1,1,1,1,1]
   }
   n8 = "Transform (Modify) Bone - Bone: dyn_neck_fix_left"(ComponentPose=@n7, Alpha=1, Translation="1.3,0,0", Rotation="0,0,0", Scale="1,1,1") {
    rotationMode=BMM_Additive
    rotationSpace=BCS_BoneSpace
    translationMode=BMM_Additive
    translationSpace=BCS_BoneSpace
    alphaBoolBlend {blendOption=Linear}
    boneToModify {boneName=dyn_neck_fix_left}
    translation {x=1.3}
   }
   Output Pose(Result=Component To Local(ComponentPose=@n8)) {layerGroup=None}
  ]}
 EventGraph {nodes [
   Event BlueprintUpdateAnimation()
   > Set LOD_alpha(LOD_alpha:bool="integer > integer"(A=Get CurrentLOD(), B=1)) -> (Output_Get:bool)
  ]}
]
variables [LOD_alpha {type=boolean}]
